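# Control Rig for the female Lucid Azalea face accessory that adjusts horn and back-hair bones from head pose.
# Three spherical pose readers on the head, relative to spine_05, produce right, left and up weights.
# These weights rotate the horns and back-hair master and shift a spine-parented aim helper, which the ADS flag also offsets.
# The back-hair master then aim-constrains to that helper, except while isSkydiving is true.
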ControlRigBlueprint /BRCosmetics/Characters/Player/Female/Medium/Bodies/F_MED_LucidAzalea/Meshes/Parts/F_MED_LucidAzalea_FaceAcc_CtrlRig
graphs [
 RigVMFunctionLibrary {links [], nodes []}
 RigVMModel {nodes [
   RigUnit_BeginExecution = &RigUnit_BeginExecution()
   SequenceAggregate = RigVMAggregateNode(ExecuteContext=@RigUnit_BeginExecution.ExecutePin)
   n1 = &RigUnit_SphericalPoseReader(
    ActiveRegionScaleFactors=(PositiveWidth=2.5,NegativeWidth=2,PositiveHeight=1.1,NegativeHeight=1.1)
    ActiveRegionSize=0.1
    Debug=(bDrawDebug=false,bDraw2D=false,bDrawLocalAxes=true,DebugScale=25,DebugSegments=20,DebugThickness=0.25)
    DriverAxis=(X=0,Y=1,Z=0)
    DriverItem=(Type=Bone,Name="head")
    FalloffRegionScaleFactors=(PositiveWidth=1,NegativeWidth=1,PositiveHeight=0.4,NegativeHeight=0.4)
    FalloffSize=0.4
    FlipHeightScaling=false
    FlipWidthScaling=false
    OptionalParentItem=(Type=Bone,Name="spine_05")
    RotationOffset=(X=55,Y=0,Z=0)
    ExecutePin=@SequenceAggregate.A
   )
   > RigVMVariableNode(Value=@n1.OutputParam) {variable=HeadRight_Adjust}
   > n2 = &RigUnit_SphericalPoseReader(
    ActiveRegionScaleFactors=(PositiveWidth=2.5,NegativeWidth=2,PositiveHeight=1.1,NegativeHeight=1.1)
    ActiveRegionSize=0.1
    Debug=(bDrawDebug=false,bDraw2D=false,bDrawLocalAxes=true,DebugScale=25,DebugSegments=20,DebugThickness=0.25)
    DriverAxis=(X=0,Y=1,Z=0)
    DriverItem=(Type=Bone,Name="head")
    FalloffRegionScaleFactors=(PositiveWidth=1,NegativeWidth=1,PositiveHeight=0.4,NegativeHeight=0.4)
    FalloffSize=0.4
    FlipHeightScaling=false
    FlipWidthScaling=false
    OptionalParentItem=(Type=Bone,Name="spine_05")
    RotationOffset=(X=140,Y=0,Z=0)
   )
   > RigVMVariableNode(Value=@n2.OutputParam) {variable=HeadLeft_Adjust}
   > n3 = &RigUnit_SphericalPoseReader(
    ActiveRegionScaleFactors=(PositiveWidth=1,NegativeWidth=1,PositiveHeight=1,NegativeHeight=1)
    ActiveRegionSize=0.15
    Debug=(bDrawDebug=true,bDraw2D=false,bDrawLocalAxes=true,DebugScale=25,DebugSegments=20,DebugThickness=0.25)
    DriverAxis=(X=0,Y=1,Z=0)
    DriverItem=(Type=Bone,Name="head")
    FalloffRegionScaleFactors=(PositiveWidth=1.5,NegativeWidth=1.5,PositiveHeight=1.5,NegativeHeight=1.5)
    FalloffSize=0.4
    FlipHeightScaling=false
    FlipWidthScaling=false
    OptionalParentItem=(Type=Bone,Name="spine_05")
    RotationOffset=(X=90,Y=0,Z=-50)
   )
   > RigVMVariableNode(Value=@n3.OutputParam) {variable=HeadUp_Adjust}
   SequenceAggregate_1 = &RigVMFunction_Sequence(ExecuteContext=@SequenceAggregate.B)
   n4 = &RigVMFunction_MathQuaternionFromEuler(Euler=(X=0,Y=0,Z=-35), RotationOrder=XYZ)
   &RigUnit_ModifyTransforms(
    ItemToModify=((Item=(Type=Bone,Name="dyn_horn_l_01"),Transform=(Rotation=(X=0,Y=0,Z=0,W=1),Translation=(X=0,Y=0,Z=0),Scale3D=(X=1,Y=1,Z=1))))
    WeightMinimum=0
    ItemToModify.0.Transform.Rotation=@n4.Result
    ExecutePin=@SequenceAggregate_1.A
    Weight=RigVMVariableNode() {variable=HeadUp_Adjust}.Value
   )
   > &RigUnit_ModifyTransforms(
    ItemToModify=((Item=(Type=Bone,Name="dyn_horn_l_02"),Transform=(Rotation=(X=0,Y=0,Z=0,W=1),Translation=(X=0,Y=0,Z=0),Scale3D=(X=1,Y=1,Z=1))))
    WeightMinimum=0
    ItemToModify.0.Transform.Rotation=@n4.Result
    Weight=RigVMVariableNode() {variable=HeadUp_Adjust}.Value
   )
   > &RigUnit_ModifyTransforms(
    ItemToModify=((Item=(Type=Bone,Name="dyn_horn_r_01"),Transform=(Rotation=(X=0,Y=0,Z=0,W=1),Translation=(X=0,Y=0,Z=0),Scale3D=(X=1,Y=1,Z=1))))
    WeightMinimum=0
    ItemToModify.0.Transform.Rotation=@n4.Result
    Weight=RigVMVariableNode() {variable=HeadUp_Adjust}.Value
   )
   > &RigUnit_ModifyTransforms(
    ItemToModify=((Item=(Type=Bone,Name="dyn_horn_r_02"),Transform=(Rotation=(X=0,Y=0,Z=0,W=1),Translation=(X=0,Y=0,Z=0),Scale3D=(X=1,Y=1,Z=1))))
    WeightMinimum=0
    ItemToModify.0.Transform.Rotation=@n4.Result
    Weight=RigVMVariableNode() {variable=HeadUp_Adjust}.Value
   )
   &RigUnit_ModifyTransforms(
    ItemToModify=((Item=(Type=Bone,Name="dyn_hair_aim_helper"),Transform=(Rotation=(X=0,Y=0,Z=0,W=1),Translation=(X=10,Y=0,Z=0),Scale3D=(X=1,Y=1,Z=1))))
    WeightMinimum=0
    ExecutePin=@SequenceAggregate_1.B
    Weight=RigVMVariableNode() {variable=HeadRight_Adjust}.Value
   )
   > &RigUnit_ModifyTransforms(
    ItemToModify=((Item=(Type=Bone,Name="dyn_hair_aim_helper"),Transform=(Rotation=(X=0,Y=0,Z=0,W=1),Translation=(X=-10,Y=0,Z=0),Scale3D=(X=1,Y=1,Z=1))))
    WeightMinimum=0
    Weight=RigVMVariableNode() {variable=HeadLeft_Adjust}.Value
   )
   RigVMFunction_MathQuaternionFromEuler = &RigVMFunction_MathQuaternionFromEuler(Euler=(X=0,Y=0,Z=-10), RotationOrder=XYZ)
   > &RigUnit_ModifyTransforms(
    ItemToModify=((Item=(Type=Bone,Name="dyn_hair_back_mstr"),Transform=(Rotation=(X=0,Y=0,Z=0,W=1),Translation=(X=0,Y=0,Z=0),Scale3D=(X=1,Y=1,Z=1))))
    WeightMinimum=0
    ItemToModify.0.Transform.Rotation=@RigVMFunction_MathQuaternionFromEuler.Result
    Weight=RigVMVariableNode() {variable=HeadUp_Adjust}.Value
   )
   > &RigUnit_ModifyTransforms(
    ItemToModify=((Item=(Type=Bone,Name="dyn_hair_aim_helper"),Transform=(Rotation=(X=0,Y=0,Z=0,W=1),Translation=(X=0,Y=-8,Z=4),Scale3D=(X=1,Y=1,Z=1))))
    WeightMinimum=0
    Weight=RigVMVariableNode() {variable=HeadUp_Adjust}.Value
   )
   > &RigUnit_ModifyTransforms(
    ItemToModify=((Item=(Type=Bone,Name="dyn_hair_aim_helper"),Transform=(Rotation=(X=-0.1736,Y=0,Z=0,W=0.9848),Translation=(X=-2,Y=-2,Z=0),Scale3D=(X=1,Y=1,Z=1))))
    WeightMinimum=0
    Weight=&RigVMFunction_MathBoolToFloat(Value=RigVMVariableNode() {variable=ADS}.Value).Result
   )
   RigVMFunction_MathBoolNot = &RigVMFunction_MathBoolNot(Value=RigVMVariableNode() {variable=isSkydiving}.Value)
   RigVMFunction_MathBoolToFloat = &RigVMFunction_MathBoolToFloat(Value=@RigVMFunction_MathBoolNot.Result)
   &RigUnit_AimConstraintLocalSpaceOffset(
    Child=(Type=Bone,Name="dyn_hair_back_mstr")
    Filter=(bX=true,bY=true,bZ=true)
    Parents=((Item=(Type=Bone,Name="dyn_hair_aim_helper"),Weight=1))
    WorldUp=(Target=(X=0,Y=1,Z=0),Kind=Location,Space=(Type=Bone,Name="dyn_hair_aim_helper"))
    bMaintainOffset=true
    Weight=@RigVMFunction_MathBoolToFloat.Result
    ExecutePin=@SequenceAggregate.C
   )
  ]}
 SequenceAggregate_1_ContainedGraph {nodes [
   n1 = RigVMFunctionEntryNode()
   SequenceAggregate = &RigVMFunction_Sequence(ExecuteContext=@n1.ExecuteContext)
   SequenceAggregate_1 = &RigVMFunction_Sequence(ExecuteContext=@SequenceAggregate.B)
   RigVMFunctionReturnNode(A=@SequenceAggregate.A, B=@SequenceAggregate_1.A, C=@SequenceAggregate_1.B)
  ]}
]
hierarchy [
 root {loc=<0 0 0>, rot=<0 0 0 1>}
 pelvis {parent=root, loc=<0 -0.5301 98.0571>, rot=<0.0127 0.707 -0.0127 -0.707>}
 spine_01 {parent=pelvis, loc=<6.1208 -0.9328 0>, rot=<0 0 0.034 -0.9994>}
 spine_02 {parent=spine_01, loc=<6.5996 -0.2095 0>, rot=<0 0 0.0238 -0.9997>}
 spine_03 {parent=spine_02, loc=<6.4914 -0.1237 0>, rot=<0 0 0.0149 -0.9999>}
 spine_04 {parent=spine_03, loc=<6.9598 0.0794 0>, rot=<0 0 -0.0132 -0.9999>}
 spine_05 {parent=spine_04, loc=<6.9987 0 0>, rot=<0 0 -0.0015 -1>}
 neck_01 {parent=spine_05, loc=<11.3138 0.1529 0>, rot=<0 0 -0.2003 -0.9797>}
 neck_02 {parent=neck_01, loc=<4.5627 0 0>, rot=<0 0 0.0504 -0.9987>}
 head {parent=neck_02, loc=<5.3914 0 0>, rot=<0 0 0.115 -0.9934>}
 dyn_hair_side_l_mstr {parent=head, loc=<-2.2964 3.2145 -6.7828>, rot=<0.0044 -0.0472 0.9954 0.0837>}
 dyn_hair_side_l_01 {parent=dyn_hair_side_l_mstr, loc=<0 0 0>, rot=<0 0 0 1>}
 dyn_hair_side_l_02 {parent=dyn_hair_side_l_01, loc=<8.9306 0 0>, rot=<-0.0829 -0.0252 0.1177 -0.9893>}
 dyn_hair_side_l_03 {parent=dyn_hair_side_l_02, loc=<10.072 0 0>, rot=<0.0608 0.0331 0.0854 -0.9939>}
 dyn_hair_side_r_mstr {parent=head, loc=<-2.1981 3.3647 6.5126>, rot=<0.0035 -0.0475 0.9967 0.0656>}
 dyn_hair_side_r_01 {parent=dyn_hair_side_r_mstr, loc=<0 0 0>, rot=<0 0 0 1>}
 dyn_hair_side_r_02 {parent=dyn_hair_side_r_01, loc=<8.9142 0 0>, rot=<-0.006 -0.0018 0.1547 -0.9879>}
 dyn_hair_side_r_03 {parent=dyn_hair_side_r_02, loc=<12.0037 0 0>, rot=<-0.1612 -0.0938 0.0752 -0.9796>}
 dyn_hair_back_mstr {parent=head, loc=<10.4335 -8.5594 0.0321>, rot=<0.0357 -0.999 0.0266 0.0007>}
 dyn_hair_back_01 {parent=dyn_hair_back_mstr, loc=<9.4128 0 0>, rot=<0 0 0.0845 -0.9964>}
 dyn_hair_back_02 {parent=dyn_hair_back_01, loc=<12.8206 0 0>, rot=<0 0 -0.035 -0.9994>}
 dyn_hair_back_03 {parent=dyn_hair_back_02, loc=<16.3605 0 0>, rot=<0 0 -0.0293 -0.9996>}
 dyn_hair_back_04 {parent=dyn_hair_back_03, loc=<15.5022 0 0>, rot=<0 0 0 1>}
 dyn_bangs_01 {parent=head, loc=<9.6666 9.8604 -1.2184>, rot=<-0.0388 0.0586 0.9973 0.0187>}
 dyn_bangs_02 {parent=head, loc=<9.595 9.636 1.6298>, rot=<0.0481 -0.1897 0.9806 0.0069>}
 dyn_bangs_03 {parent=head, loc=<9.5922 8.8952 2.7242>, rot=<0.0874 -0.3041 0.9482 -0.0283>}
 dyn_bangs_04 {parent=head, loc=<10.0419 8.1224 3.5583>, rot=<0.1412 -0.3588 0.9225 -0.0167>}
 dyn_horn_l_01 {parent=head, loc=<3.8558 -3.4338 -7.722>, rot=<0.669 -0.6741 0.2241 0.2188>}
 dyn_horn_l_02 {parent=head, loc=<1.0931 -2.8062 -6.8597>, rot=<0.5837 -0.7899 0.1523 0.1106>}
 dyn_horn_r_01 {parent=head, loc=<3.8558 -3.4338 7.722>, rot=<0.669 -0.6741 -0.2241 -0.2188>}
 dyn_horn_r_02 {parent=head, loc=<1.0931 -2.8062 6.8597>, rot=<0.5837 -0.7899 -0.1523 -0.1106>}
 clavicle_l {parent=spine_05, loc=<5.4831 -0.1053 -3.2946>, rot=<0.0465 -0.7289 -0.009 -0.6829>}
 upperarm_l {parent=clavicle_l, loc=<9.8658 -1.2136 -0.4105>, rot=<0.0164 -0.3245 -0.0269 -0.9454>}
 lowerarm_l {parent=upperarm_l, loc=<26.465 -0.8345 0.259>, rot=<0.0036 0.0161 -0.2298 -0.9731>}
 clavicle_r {parent=spine_05, loc=<5.4831 -0.1053 3.2946>, rot=<0.7289 0.0465 -0.6829 0.009>}
 upperarm_r {parent=clavicle_r, loc=<-9.8658 1.2136 0.4105>, rot=<0.0164 -0.3245 -0.0269 -0.9454>}
 lowerarm_r {parent=upperarm_r, loc=<-26.465 0.8345 -0.259>, rot=<0.0036 0.0161 -0.2298 -0.9731>}
 dyn_hair_aim_helper {parent=spine_05, loc=<5.5375 -10.6755 0>, rot=<-0.0093 -0.707 0.0093 -0.707>}
]
variables [
 ADS {type=bool, default=False}
 HeadLeft_Adjust {type=double, default=0}
 HeadRight_Adjust {type=double, default=0}
 HeadUp_Adjust {type=double, default=0}
 isSkydiving {type=bool, default=False}
]
