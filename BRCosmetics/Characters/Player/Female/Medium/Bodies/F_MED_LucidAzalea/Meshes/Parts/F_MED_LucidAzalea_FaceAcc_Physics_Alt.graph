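# Alternate physics asset for the LucidAzalea female face accessory mesh, defining capsule bodies and joints for its dynamic hair.
# A four-bone back chain has very tight swing and hangs from a master bone freely jointed to the head.
# Two three-bone side chains attach to the head with looser swing and narrow twist.
# Torso, neck, head and arm capsules form a body proxy linked by angularly free constraints.
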PhysicsAsset /BRCosmetics/Characters/Player/Female/Medium/Bodies/F_MED_LucidAzalea/Meshes/Parts/F_MED_LucidAzalea_FaceAcc_Physics_Alt
bodies {
 clavicle_l [{shapeType=Capsule, length=17.3841, radius=5.977, shapeName=clavicle_l_capsule, center=<3.6808 -1.0427 0>, rotation=<rot -90 180 180>}]
 clavicle_r [{shapeType=Capsule, length=17.3841, radius=5.977, shapeName=clavicle_r_capsule, center=<-3.6808 1.0427 0>, rotation=<rot -90 0 0>}]
 dyn_hair_back_01 [{shapeType=Capsule, length=13.4051, radius=4.3772, shapeName=dyn_hair_back_01_capsule, center=<6.9063 1.6597 0>}]
 dyn_hair_back_02 [{shapeType=Capsule, length=16.2676, radius=5.303, shapeName=dyn_hair_back_02_capsule2, center=<8.7398 -1.9611 0>}]
 dyn_hair_back_03 [
  {shapeType=Capsule, length=16.2634, radius=6.1603, shapeName=dyn_hair_back_03_capsule2, center=<2.9953 -3.2876 0>}
  {shapeType=Capsule, length=15.9148, radius=6.5089, shapeName=dyn_hair_back_03_capsule, center=<15.4673 -2.5638 0>}
 ]
 dyn_hair_back_04 [{shapeType=Capsule, length=15.282, radius=7.2316, shapeName=dyn_hair_back_04_capsule, center=<13.7392 -2.0499 0>}]
 dyn_hair_back_mstr [{shapeType=Capsule, length=0.505, radius=0.505, shapeName=dyn_hair_back_mstr_capsule, rotation=<rot 0 -90 0>}]
 dyn_hair_side_l_01 [{shapeType=Capsule, length=5.1631, radius=1.1341, shapeName=dyn_hair_side_l_01_capsule, center=<4.6758 -0.3723 0>, rotation=<rot 90 0 0>}]
 dyn_hair_side_l_02 [{shapeType=Capsule, length=7.5379, radius=1.029, shapeName=dyn_hair_side_l_02_capsule, center=<5.4565 -0.3723 0>, rotation=<rot 90 0 0>}]
 dyn_hair_side_l_03 [{shapeType=Capsule, length=9.0552, radius=0.9701, shapeName=dyn_hair_side_l_03_capsule, center=<4.6758 -0.7585 0>, rotation=<rot 90 0 0>}]
 dyn_hair_side_r_01 [{shapeType=Capsule, length=7.275, radius=1.2421, shapeName=dyn_hair_side_r_01_capsule, center=<4.6758 0 0>, rotation=<rot 90 0 0>}]
 dyn_hair_side_r_02 [{shapeType=Capsule, length=6.7579, radius=1.2421, shapeName=dyn_hair_side_r_02_capsule, center=<4.6758 0 0>, rotation=<rot 90 0 0>}]
 dyn_hair_side_r_03 [{
   shapeType=Capsule
   length=8.7831
   radius=1.2421
   shapeName=dyn_hair_side_r_03_capsule
   center=<2.8451 0.0126 0>
   rotation=<rot 90 180 -177.5324>
  }]
 head [{shapeType=Capsule, length=2.7144, radius=1.2155, shapeName=head_capsule}]
 lowerarm_l [{shapeType=Capsule, length=21.1898, radius=4.9505, shapeName=lowerarm_l_capsule, center=<8.4834 0 0>, rotation=<rot 90 0 0>}]
 lowerarm_r [{shapeType=Capsule, length=21.1898, radius=4.9505, shapeName=lowerarm_r_capsule, center=<-8.4834 0 0>, rotation=<rot 90 0 0>}]
 neck_01 [{shapeType=Capsule, length=5.5149, radius=4.3835, shapeName=neck_01_capsule, center=<3.146 -0.154 0>}]
 pelvis [{shapeType=Capsule, length=10.4629, radius=11.6794, shapeName=pelvis_capsule, center=<-1.9311 -0.1695 0>, rotation=<rot 0 -90 0>}]
 spine_01 [{shapeType=Capsule, length=9.7152, radius=9.2122, shapeName=spine_01_capsule, center=<0 0.0692 0>, rotation=<rot 0 -90 0>}]
 spine_02 [{shapeType=Capsule, length=9.7152, radius=9.2122, shapeName=spine_02_capsule, center=<0 1.4257 0>, rotation=<rot 0 -90 0>}]
 spine_03 [{shapeType=Capsule, length=9.7152, radius=9.2122, shapeName=spine_03_capsule, center=<0 1.4257 0>, rotation=<rot 0 -90 0>}]
 spine_04 [{shapeType=Capsule, length=36.0357, radius=9.3058, shapeName=spine_04_capsule, center=<-0.5233 2.1564 0>, rotation=<rot 0 -90 0>}]
 spine_05 [{shapeType=Capsule, length=31.3614, radius=5.3275, shapeName=spine_05_capsule, center=<2.0309 -0.2571 0>, rotation=<rot 0 -90 0>}]
 upperarm_l [{shapeType=Capsule, length=21.5741, radius=5.7732, shapeName=upperarm_l_capsule, center=<10.2218 0 0>, rotation=<rot 90 0 0>}]
 upperarm_r [{shapeType=Capsule, length=21.5741, radius=5.7732, shapeName=upperarm_l_capsule, center=<-10.2218 0 0>, rotation=<rot 90 0 0>}]
}
constraints [
 {bone1Name=dyn_hair_back_02, bone2Name=dyn_hair_back_01, swing1LimitDegrees=1, swing2LimitDegrees=1, twistLimitDegrees=8}
 {bone1Name=dyn_hair_back_03, bone2Name=dyn_hair_back_02, swing1LimitDegrees=1, swing2LimitDegrees=1, twistLimitDegrees=8}
 {bone1Name=dyn_hair_back_04, bone2Name=dyn_hair_back_03, swing1LimitDegrees=1, swing2LimitDegrees=1, twistLimitDegrees=8}
 {bone1Name=dyn_hair_side_l_02, bone2Name=dyn_hair_side_l_01, swing1LimitDegrees=45, swing2LimitDegrees=45, twistLimitDegrees=2}
 {bone1Name=dyn_hair_side_l_03, bone2Name=dyn_hair_side_l_02, swing1LimitDegrees=45, swing2LimitDegrees=45, twistLimitDegrees=2}
 {bone1Name=dyn_hair_side_r_02, bone2Name=dyn_hair_side_r_01, swing1LimitDegrees=45, swing2LimitDegrees=45, twistLimitDegrees=2}
 {bone1Name=dyn_hair_side_r_03, bone2Name=dyn_hair_side_r_02, swing1LimitDegrees=45, swing2LimitDegrees=45, twistLimitDegrees=2}
 {bone1Name=dyn_hair_side_l_01, bone2Name=head, swing1LimitDegrees=24, swing2LimitDegrees=24, twistLimitDegrees=2}
 {bone1Name=dyn_hair_side_r_01, bone2Name=head, swing1LimitDegrees=24, swing2LimitDegrees=24, twistLimitDegrees=2}
 {bone1Name=dyn_hair_back_01, bone2Name=dyn_hair_back_mstr, swing1LimitDegrees=1, swing2LimitDegrees=1, twistLimitDegrees=0}
 {
  bone1Name=head
  bone2Name=neck_01
  swing1LimitDegrees=45
  swing1Motion=Free
  swing2LimitDegrees=45
  swing2Motion=Free
  twistLimitDegrees=45
  twistMotion=Free
 }
 {
  bone1Name=lowerarm_l
  bone2Name=upperarm_l
  swing1LimitDegrees=45
  swing1Motion=Free
  swing2LimitDegrees=45
  swing2Motion=Free
  twistLimitDegrees=45
  twistMotion=Free
 }
 {
  bone1Name=lowerarm_r
  bone2Name=upperarm_r
  swing1LimitDegrees=45
  swing1Motion=Free
  swing2LimitDegrees=45
  swing2Motion=Free
  twistLimitDegrees=45
  twistMotion=Free
 }
 {
  bone1Name=neck_01
  bone2Name=spine_05
  swing1LimitDegrees=45
  swing1Motion=Free
  swing2LimitDegrees=45
  swing2Motion=Free
  twistLimitDegrees=45
  twistMotion=Free
 }
 {
  bone1Name=upperarm_l
  bone2Name=clavicle_l
  swing1LimitDegrees=45
  swing1Motion=Free
  swing2LimitDegrees=45
  swing2Motion=Free
  twistLimitDegrees=45
  twistMotion=Free
 }
 {
  bone1Name=upperarm_r
  bone2Name=clavicle_r
  swing1LimitDegrees=45
  swing1Motion=Free
  swing2LimitDegrees=45
  swing2Motion=Free
  twistLimitDegrees=45
  twistMotion=Free
 }
 {
  bone1Name=spine_01
  bone2Name=pelvis
  swing1LimitDegrees=45
  swing1Motion=Free
  swing2LimitDegrees=45
  swing2Motion=Free
  twistLimitDegrees=45
  twistMotion=Free
 }
 {
  bone1Name=spine_02
  bone2Name=spine_01
  swing1LimitDegrees=45
  swing1Motion=Free
  swing2LimitDegrees=45
  swing2Motion=Free
  twistLimitDegrees=45
  twistMotion=Free
 }
 {
  bone1Name=spine_03
  bone2Name=spine_02
  swing1LimitDegrees=45
  swing1Motion=Free
  swing2LimitDegrees=45
  swing2Motion=Free
  twistLimitDegrees=45
  twistMotion=Free
 }
 {
  bone1Name=spine_04
  bone2Name=spine_03
  swing1LimitDegrees=45
  swing1Motion=Free
  swing2LimitDegrees=45
  swing2Motion=Free
  twistLimitDegrees=45
  twistMotion=Free
 }
 {
  bone1Name=spine_05
  bone2Name=spine_04
  swing1LimitDegrees=45
  swing1Motion=Free
  swing2LimitDegrees=45
  swing2Motion=Free
  twistLimitDegrees=45
  twistMotion=Free
 }
 {
  bone1Name=dyn_hair_back_mstr
  bone2Name=head
  swing1LimitDegrees=45
  swing1Motion=Free
  swing2LimitDegrees=45
  swing2Motion=Free
  twistLimitDegrees=45
  twistMotion=Free
 }
 {
  bone1Name=clavicle_r
  bone2Name=spine_05
  swing1LimitDegrees=45
  swing1Motion=Free
  swing2LimitDegrees=45
  swing2Motion=Free
  twistLimitDegrees=45
  twistMotion=Free
 }
 {
  bone1Name=clavicle_l
  bone2Name=spine_05
  swing1LimitDegrees=45
  swing1Motion=Free
  swing2LimitDegrees=45
  swing2Motion=Free
  twistLimitDegrees=45
  twistMotion=Free
 }
]
