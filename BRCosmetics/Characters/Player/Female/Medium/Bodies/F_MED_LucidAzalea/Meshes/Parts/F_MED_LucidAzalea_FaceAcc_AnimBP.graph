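# Post-process style AnimBlueprint for the LucidAzalea face accessory, animating hair and bangs atop the attached parent's pose.
# Backbling state selects between two Control Rig and RigidBody chains: the default adds AnimDynamics bangs, the backbling one uses alternate physics, constraints and head-rotation-driven bone offsets.
# The event graph computes head-rotation alphas, flips gravity upward and lowers physics alpha while skydiving, and feeds WPO motion and wind to dynamic materials.
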AnimBlueprint /BRCosmetics/Characters/Player/Female/Medium/Bodies/F_MED_LucidAzalea/Meshes/Parts/F_MED_LucidAzalea_FaceAcc_AnimBP
target_skeleton=&/F_MED_LucidAzalea_FaceAcc_Skeleton
graphs [
 AnimGraph {nodes [
   n1 = Control Rig(Source=Copy Pose From Mesh() {bUseAttachedParent=true}) {
    alphaBoolBlend {blendOption=Linear}
    controlRigAssetReference {blueprintRigClass=&/F_MED_LucidAzalea_FaceAcc_CtrlRig.F_MED_LucidAzalea_FaceAcc_CtrlRig_C}
   }
   n2 = RigidBody(ComponentPose=Local To Component(LocalPose=@n1), Alpha=Get SkydiveAlpha() -> (SkydiveAlpha:double), ExternalForce=Get GravityOverride()) {
    bOverrideWorldGravity=true
    overridePhysicsAsset=&/F_MED_LucidAzalea_FaceAcc_Physics_Alt
    simulationSpace=BaseBoneSpace
    alphaBoolBlend {blendOption=Linear}
    baseBoneRef {boneName=spine_05}
    componentAppliedLinearAccClamp=<1000 1000 1000>
    componentLinearAccScale=<0.2 0.2 0.2>
    componentLinearVelScale=<0.2 0.2 0.2>
    overrideWorldGravity {z=-980}
    simSpaceSettings {maxAngularAcceleration=100, maxAngularVelocity=3, maxLinearAcceleration=250, maxLinearVelocity=500, velocityScaleZ=0, worldAlpha=1}
   }
   n3 = "Constraint - dyn_hair_back_02 (PP)"!(ComponentPose=@n2, Alpha=1, ConstraintWeights_0=0.8, ConstraintWeights_1:float=0.2) -> (Pose) {
    alphaBoolBlend {blendOption=Linear}
    boneToModify {boneName=dyn_hair_back_02}
    constraintSetup [{targetBone {boneName=spine_05}}, {targetBone {boneName=head}}]
    constraintWeights [1,1]
   }
   n4 = "Constraint - dyn_hair_back_04 (PP)"!(ComponentPose=@n3, Alpha=1, ConstraintWeights_0=0.8, ConstraintWeights_1:float=0.2) -> (Pose) {
    alphaBoolBlend {blendOption=Linear}
    boneToModify {boneName=dyn_hair_back_04}
    constraintSetup [{targetBone {boneName=pelvis}}, {targetBone {boneName=head}}]
    constraintWeights [1,1]
   }
   n5 = "Transform (Modify) Bone - Bone: dyn_hair_back_01"!(ComponentPose=@n4, bAlphaBoolEnabled:bool=True, Translation="4,0,0", Rotation="0,0,0", Scale="1,1,1") -> (Pose) {
    alphaInputType=Bool
    rotationMode=BMM_Additive
    rotationSpace=BCS_BoneSpace
    translationMode=BMM_Additive
    translationSpace=BCS_BoneSpace
    alphaBoolBlend {blendOption=Linear}
    boneToModify {boneName=dyn_hair_back_01}
    translation {x=4}
   }
   n6 = "Transform (Modify) Bone - Bone: dyn_hair_back_02"!(ComponentPose=@n5, bAlphaBoolEnabled:bool=True, Translation="0,0,0", Rotation="0,0,0", Scale="1,1,1") -> (Pose) {
    alphaInputType=Bool
    rotationMode=BMM_Additive
    rotationSpace=BCS_BoneSpace
    translationMode=BMM_Additive
    translationSpace=BCS_BoneSpace
    alphaBoolBlend {blendOption=Linear}
    boneToModify {boneName=dyn_hair_back_02}
   }
   n7 = "Transform (Modify) Bone - Bone: dyn_hair_back_03"!(ComponentPose=@n6, bAlphaBoolEnabled:bool=True, Translation="0,6,0", Rotation="0,0,0", Scale="1,1,1") -> (Pose) {
    alphaInputType=Bool
    rotationMode=BMM_Additive
    rotationSpace=BCS_BoneSpace
    translationMode=BMM_Additive
    translationSpace=BCS_BoneSpace
    alphaBoolBlend {blendOption=Linear}
    boneToModify {boneName=dyn_hair_back_03}
    translation {y=6}
   }
   n8 = "Transform (Modify) Bone - Bone: dyn_hair_back_04"!(ComponentPose=@n7, bAlphaBoolEnabled:bool=True, Translation="0,-4,0", Rotation="0,-2,0", Scale="1,1,1") -> (Pose) {
    alphaInputType=Bool
    rotationMode=BMM_Additive
    rotationSpace=BCS_BoneSpace
    translationMode=BMM_Additive
    translationSpace=BCS_BoneSpace
    alphaBoolBlend {blendOption=Linear}
    boneToModify {boneName=dyn_hair_back_04}
    rotation {yaw=-2}
    translation {y=-4}
   }
   n9 = "Transform (Modify) Bone - Bone: dyn_hair_back_03"!(ComponentPose=@n8, Alpha:float=Get head Rot Fwd() -> (head Rot Fwd:float), Translation="0,4,0", Rotation="0,10,0", Scale="1,1,1") -> (Pose) {
    rotationMode=BMM_Additive
    rotationSpace=BCS_BoneSpace
    translationMode=BMM_Additive
    translationSpace=BCS_BoneSpace
    alphaBoolBlend {blendOption=Linear}
    boneToModify {boneName=dyn_hair_back_03}
    rotation {yaw=10}
    translation {y=4}
   }
   n10 = "Transform (Modify) Bone - Bone: dyn_hair_back_04"!(ComponentPose=@n9, Alpha:float=Get head Rot Back() -> (head Rot Back:float), Translation="0,6,0", Rotation="0,-12,10", Scale="1,1,1") -> (Pose) {
    rotationMode=BMM_Additive
    rotationSpace=BCS_BoneSpace
    translationMode=BMM_Additive
    translationSpace=BCS_BoneSpace
    alphaBoolBlend {blendOption=Linear}
    boneToModify {boneName=dyn_hair_back_04}
    rotation {roll=10, yaw=-12}
    translation {y=6}
   }
   n11 = "Transform (Modify) Bone - Bone: dyn_hair_back_02"!(ComponentPose=@n10, Alpha:float=Get head Rot Back(), Translation="0,-10,0", Rotation="0,0,0", Scale="1,1,1") -> (Pose) {
    rotationMode=BMM_Additive
    rotationSpace=BCS_BoneSpace
    translationMode=BMM_Additive
    translationSpace=BCS_BoneSpace
    alphaBoolBlend {blendOption=Linear}
    boneToModify {boneName=dyn_hair_back_02}
    translation {y=-10}
   }
   n12 = "Transform (Modify) Bone - Bone: dyn_hair_back_03"(ComponentPose=@n11, Alpha=Get head Rot left() -> (head Rot left:float), Translation="0,0,-4", Rotation="0,0,0", Scale="1,1,1") {
    rotationMode=BMM_Additive
    rotationSpace=BCS_BoneSpace
    translationMode=BMM_Additive
    translationSpace=BCS_BoneSpace
    alphaBoolBlend {blendOption=Linear}
    boneToModify {boneName=dyn_hair_back_03}
    translation {z=-4}
   }
   n13 = "Transform (Modify) Bone - Bone: dyn_hair_back_03"(ComponentPose=@n12, Alpha=Get head Rot right() -> (head Rot right:float), Translation="0,0,5", Rotation="0,0,0", Scale="1,1,1") {
    rotationMode=BMM_Additive
    rotationSpace=BCS_BoneSpace
    translationMode=BMM_Additive
    translationSpace=BCS_BoneSpace
    alphaBoolBlend {blendOption=Linear}
    boneToModify {boneName=dyn_hair_back_03}
    translation {z=5}
   }
   n14 = "Transform (Modify) Bone - Bone: dyn_hair_back_04"!(ComponentPose=@n13, bAlphaBoolEnabled:bool=Get bIsSkydiveFloating(), Translation="0,-5,0", Rotation="0,0,0", Scale="1,1,1") -> (Pose) {
    alphaInputType=Bool
    rotationMode=BMM_Additive
    rotationSpace=BCS_BoneSpace
    translationMode=BMM_Additive
    translationSpace=BCS_BoneSpace
    alphaBoolBlend {blendInTime=0.2, blendOption=Linear, blendOutTime=0.2}
    boneToModify {boneName=dyn_hair_back_04}
    translation {y=-5}
   }
   n15 = Component To Local(ComponentPose=@n14)
   n16 = Control Rig!(Source=Copy Pose From Mesh() {bUseAttachedParent=true}, Alpha, isSkydiving:bool=Get bIsSkyDiving(), ADS:bool=Get ADS() -> (ADS:bool)) -> (Pose) {
    alphaBoolBlend {blendOption=Linear}
    controlRigAssetReference {blueprintRigClass=&F_MED_LucidAzalea_FaceAcc_CtrlRig_C}
   }
   n17 = Anim Dynamics - Bone: dyn_bangs_01(ComponentPose=Local To Component(LocalPose=@n16), Alpha=1, ExternalForce="0,0,0") {
    angularSpringConstant=10
    bAngularSpring=true
    bUseGravityOverride=true
    simulationSpace=RootRelative
    alphaBoolBlend {blendOption=Linear}
    boundBone {boneName=dyn_bangs_01}
    componentAppliedLinearAccClamp=<1000 1000 1000>
    componentLinearAccScale=<0.3 0.3 0.3>
    componentLinearVelScale=<0.3 0.3 0.3>
    gravityOverride {z=-980}
    physicsBodyDefinitions [{
      boundBone {boneName=dyn_bangs_01}
      boxExtents=<6 6 6>
      constraintSetup {angularLimitsMax=<0 12 0>, angularLimitsMin=<0 -12 -24>, angularTarget=<1 0 0>}
      localJointOffset=<0.1 0 0>
     }]
    relativeSpaceBone {boneName=head}
   }
   n18 = Anim Dynamics - Bone: dyn_bangs_02(ComponentPose=@n17, Alpha=1, ExternalForce="0,0,0") {
    angularSpringConstant=10
    bAngularSpring=true
    bUseGravityOverride=true
    simulationSpace=RootRelative
    alphaBoolBlend {blendOption=Linear}
    boundBone {boneName=dyn_bangs_02}
    componentAppliedLinearAccClamp=<1000 1000 1000>
    componentLinearAccScale=<0.3 0.3 0.3>
    componentLinearVelScale=<0.3 0.3 0.3>
    gravityOverride {z=-980}
    physicsBodyDefinitions [{
      boundBone {boneName=dyn_bangs_02}
      boxExtents=<5 5 5>
      constraintSetup {angularLimitsMax=<0 10 0>, angularLimitsMin=<0 -10 -24>, angularTarget=<1 0 0>}
      localJointOffset=<0.15 0 0>
     }]
    relativeSpaceBone {boneName=head}
   }
   n19 = Anim Dynamics - Bone: dyn_bangs_03(ComponentPose=@n18, Alpha=1, ExternalForce="0,0,0") {
    angularSpringConstant=10
    bAngularSpring=true
    bUseGravityOverride=true
    simulationSpace=RootRelative
    alphaBoolBlend {blendOption=Linear}
    boundBone {boneName=dyn_bangs_03}
    componentAppliedLinearAccClamp=<1000 1000 1000>
    componentLinearAccScale=<0.3 0.3 0.3>
    componentLinearVelScale=<0.3 0.3 0.3>
    gravityOverride {z=-980}
    physicsBodyDefinitions [{
      boundBone {boneName=dyn_bangs_03}
      boxExtents=<5 5 5>
      constraintSetup {angularLimitsMax=<0 10 0>, angularLimitsMin=<0 -10 -24>, angularTarget=<1 0 0>}
      localJointOffset=<0.125 0 0>
     }]
    relativeSpaceBone {boneName=head}
   }
   n20 = Anim Dynamics - Bone: dyn_bangs_04(ComponentPose=@n19, Alpha=1, ExternalForce="0,0,0") {
    angularSpringConstant=10
    bAngularSpring=true
    bUseGravityOverride=true
    simulationSpace=RootRelative
    alphaBoolBlend {blendOption=Linear}
    boundBone {boneName=dyn_bangs_04}
    componentAppliedLinearAccClamp=<1000 1000 1000>
    componentLinearAccScale=<0.3 0.3 0.3>
    componentLinearVelScale=<0.3 0.3 0.3>
    gravityOverride {z=-980}
    physicsBodyDefinitions [{
      boundBone {boneName=dyn_bangs_04}
      boxExtents=<6 6 6>
      constraintSetup {angularLimitsMax=<0 10 0>, angularLimitsMin=<0 -10 -24>, angularTarget=<1 0 0>}
      localJointOffset=<0.1 0 0>
     }]
    relativeSpaceBone {boneName=head}
   }
   n21 = RigidBody!(ComponentPose=@n20, Alpha=Get SkydiveAlpha(), OverrideWorldGravity:Vector=Get GravityOverride(), ExternalForce) -> (Pose) {
    bClampLinearTranslationLimitToRefPose=true
    bOverrideWorldGravity=true
    overridePhysicsAsset=&/F_MED_LucidAzalea_FaceAcc_Physics
    simulationSpace=BaseBoneSpace
    alphaBoolBlend {blendOption=Linear}
    alphaScaleBiasClamp {bInterpResult=true}
    baseBoneRef {boneName=spine_05}
    componentAppliedLinearAccClamp=<1000 1000 1000>
    componentLinearAccScale=<0.55 0.55 0.55>
    componentLinearVelScale=<0.7 0.7 0.7>
    overrideWorldGravity {z=-980}
    simSpaceSettings {maxAngularAcceleration=100, maxAngularVelocity=3, maxLinearAcceleration=250, maxLinearVelocity=500, velocityScaleZ=0, worldAlpha=1}
   }
   n22 = "Transform (Modify) Bone - Bone: dyn_hair_back_mstr"!(ComponentPose=@n21, bAlphaBoolEnabled:bool=Get skydiveHair() -> (skydiveHair:bool), Translation="0,0,0", Rotation="0,-20,0", Scale="1,1,1") -> (Pose) {
    alphaInputType=Bool
    rotationMode=BMM_Replace
    rotationSpace=BCS_BoneSpace
    translationSpace=BCS_BoneSpace
    alphaBoolBlend {blendOption=Linear}
    boneToModify {boneName=dyn_hair_back_mstr}
    rotation {yaw=-20}
   }
   n23 = "Transform (Modify) Bone - Bone: dyn_hair_back_mstr"(ComponentPose=@n22, bAlphaBoolEnabled=True, Translation="0,2,0", Rotation="0,0,0", Scale="1,1,1") {
    alphaInputType=Bool
    rotationMode=BMM_Replace
    rotationSpace=BCS_BoneSpace
    translationMode=BMM_Additive
    translationSpace=BCS_BoneSpace
    alphaBoolBlend {blendOption=Linear}
    boneToModify {boneName=dyn_hair_back_mstr}
    translation {y=2}
   }
   n24 = Component To Local(ComponentPose=@n23)
   n25 = NOT Boolean(A=Get BackblingEquipped() -> (BackblingEquipped:bool))
   n26 = "Transform (Modify) Bone - Bone: dyn_hair_side_l_mstr"!(
    ComponentPose=Local To Component(LocalPose=Blend Poses by bool(BlendPose_0=@n24, BlendPose_1=@n15, bActiveValue=@n25) {blendType=Linear})
    bAlphaBoolEnabled:bool=Get bIsInFrontEnd()
    Translation="0,0,0"
    Rotation="6,0,0"
    Scale="1,1,1"
   ) -> (Pose) {
    alphaInputType=Bool
    rotationMode=BMM_Replace
    rotationSpace=BCS_BoneSpace
    translationSpace=BCS_BoneSpace
    alphaBoolBlend {blendOption=Linear}
    boneToModify {boneName=dyn_hair_side_l_mstr}
    rotation {pitch=6}
   }
   Output Pose(Result=Component To Local(ComponentPose=@n26)) {layerGroup=None}
   Get bIsInFrontEnd()
   Get bIsSkyDiving()
   Get bIsSkydiveDiveMode()
  ]}
 EventGraph {nodes [
   n1 = Event BlueprintUpdateAnimation()
   > n2 = Sequence()
   > Is Valid(InputObject=Get playerPawn() -> (playerPawn:Fort Player Pawn Athena Object Reference))
   > Set WPO Globals Function(
    Material=Get MatInstanceDynamic() -> (MatInstanceDynamic:Material Instance Dynamic Object Reference)
    Overlay Material=Get OverlayMID() -> (OverlayMID:Material Instance Dynamic Object Reference)
    Skeletal Mesh=Get skelMesh() -> (skelMesh:Skeletal Mesh Component Object Reference)
    "Local Noise Transform  Location"="Get Local Noise Transform  Location"() -> ("Local Noise Transform  Location":Vector)
    Local Noise Offset=Get Local Noise Offset()
   )
   > Set WPO Regional Function!(
    execute
    Material:Material Instance Dynamic Object Reference=Get MatInstanceDynamic()
    Overlay Material:Material Instance Dynamic Object Reference=Get OverlayMID()
    Skeletal Mesh:Skeletal Mesh Component Object Reference=Get skelMesh()
    Velocity Transform:Name=dyn_hair_back_03
    Velocity Interp Speed:double=10
    Collision Transform:Name=head
    Collision Transform Axis:EAxis Enum
    Invert Collision Transform Axis:bool=false
    Delta Time X:double=@n1.DeltaTimeX
    Previous Velocity Transform Location:"Vector (by ref)"=Get Previous Velocity Transform Location() -> (Previous Velocity Transform Location:Vector)
    Previous Velocity Transform Velocity:"Vector (by ref)"=Get Previous Velocity Transform Velocity() -> (Previous Velocity Transform Velocity:Vector)
    Region:int=1
   ) -> (then)
   > Set WPO Regional Function(
    Material=Get MatInstanceDynamic()
    Overlay Material=Get OverlayMID()
    Skeletal Mesh=Get skelMesh()
    Velocity Transform=dyn_hair_side_02
    Velocity Interp Speed=10
    Collision Transform=head
    Invert Collision Transform Axis=false
    Delta Time X=@n1.DeltaTimeX
    Previous Velocity Transform Location=Get Previous Velocity Transform Location 2() -> (Previous Velocity Transform Location 2:Vector)
    Previous Velocity Transform Velocity=Get Previous Velocity Transform Velocity 2() -> (Previous Velocity Transform Velocity 2:Vector)
    Region=2
   )
   > n3 = Branch(Condition=Get bIsSkydiveFloating())
   > n4 = Set WPO Pho Wind Intensity(WPO Pho Wind Intensity=1)
   SetScalarParameterValue(
    execute=[@n5.then, @n4.then]
    self=[Get OverlayMID(), Get matInstanceDynamic() -> (matInstanceDynamic:Material Instance Dynamic Object Reference)]
    ParameterName=ClothSim_WindStrength
    Value=Get WPO Pho Wind Intensity()
   )
   > Set ADS(ADS:bool=Get bIsTargeting()) -> (Output_Get:bool)
   n5 = Set WPO Pho Wind Intensity(execute=@n3.else, WPO Pho Wind Intensity=0)
   n6 = Sequence!(execute=@n2.then_1) -> (then_0, then_1, then_2:Exec, then_3:Exec, then_4:Exec, then_5:Exec)
   n7 = CheckForBackbling()
   > Set BackblingEquipped(BackblingEquipped:bool=OR Boolean(A=@n7, B=@n7.bIsBackblingACape)) -> (Output_Get:bool)
   Set head Rot left(
    execute=@n6.then_3
    head Rot left:float=GetAlphaFromReferenceTransformDelta(SkeletalMeshComponent=Get skelMesh(), BoneName=head, BaseName=spine_05, DeltaComponentType=EulerX, InputRange=(X=0,Y=55), TwistAxis="1,0,0")
   ) -> (Output_Get:float)
   Set head Rot right(
    execute=@n6.then_4
    head Rot right:float=GetAlphaFromReferenceTransformDelta(SkeletalMeshComponent=Get skelMesh(), BoneName=head, BaseName=spine_05, DeltaComponentType=EulerX, InputRange=(X=0,Y=-30), TwistAxis="1,0,0")
   ) -> (Output_Get:float)
   n8 = Branch(execute=@n6.then_5, Condition=Get bIsSkyDiving())
   > Set SkydiveAlpha(SkydiveAlpha:double=0.5) -> (Output_Get:double)
   > Set GravityOverride(GravityOverride="0,0,2000")
   > Set skydiveHair(skydiveHair:bool=true) -> (Output_Get:bool)
   Set SkydiveAlpha(execute=@n8.else, SkydiveAlpha=1)
   > Set GravityOverride(GravityOverride="0,0,-980")
   > Set skydiveHair(skydiveHair=false)
   Set head Rot Back(
    execute=@n6.then_1
    head Rot Back:float=GetAlphaFromReferenceTransformDelta(SkeletalMeshComponent=Get skelMesh(), BoneName=head, BaseName=spine_05, InputRange=(X=0,Y=-60), TwistAxis="1,0,0")
   ) -> (Output_Get:float)
   Set head Rot Fwd(
    execute=@n6.then_2
    head Rot Fwd:float=GetAlphaFromReferenceTransformDelta(SkeletalMeshComponent=Get skelMesh(), BoneName=head, BaseName=spine_05, InputRange=(X=0,Y=25), TwistAxis="1,0,0")
   ) -> (Output_Get:float)
   Event BlueprintInitializeAnimation()
   > Set SkelMesh(SkelMesh=GetOwnerSkeletalMeshComponent())
   n9 = TryGetPawnOwner()
   > Is Valid(InputObject=@n9)
   > n10 = Cast To FortPlayerPawnAthena(Object=@n9)
   > Set playerPawn(playerPawn:Fort Player Pawn Athena Object Reference=@n10) -> (Output_Get:Fort Player Pawn Athena Object Reference)
   Event OnCharacterCustomizationLoaded()
   > n11 = Set SkelMesh(SkelMesh=GetPartSkeletalMeshComponent())
   > n12 = Get Dynamic Material Instance(SkeletalMeshComponent=@n11)
   > Set MatInstanceDynamic(MatInstanceDynamic:Material Instance Dynamic Object Reference=@n12) -> (Output_Get:Material Instance Dynamic Object Reference)
   > n13 = GetOrCreateDynamicOverlayMaterial(MeshComponent=@n11, MaterialName=None)
   > Set OverlayMID(OverlayMID:Material Instance Dynamic Object Reference=@n13) -> (Output_Get:Material Instance Dynamic Object Reference)
  ]}
]
variables [
 ADS {type=boolean}
 BackblingEquipped {type=boolean}
 GravityOverride {type=Vector, default=<0 0 -980>}
 head Rot Back {type=number}
 head Rot Fwd {type=number}
 head Rot left {type=number}
 head Rot right {type=number}
 Local Noise Offset {type=Vector}
 "Local Noise Transform  Location" {type=Vector}
 MatInstanceDynamic {type=MaterialInstanceDynamic}
 OverlayMID {type=MaterialInstanceDynamic}
 playerPawn {type=FortPlayerPawnAthena}
 Previous Velocity Transform Location {type=Vector}
 Previous Velocity Transform Location 2 {type=Vector}
 Previous Velocity Transform Velocity {type=Vector}
 Previous Velocity Transform Velocity 2 {type=Vector}
 skelMesh {type=SkeletalMeshComponent}
 SkydiveAlpha {type=number}
 skydiveHair {type=boolean}
 WPO Pho Wind Intensity {type=number}
]
